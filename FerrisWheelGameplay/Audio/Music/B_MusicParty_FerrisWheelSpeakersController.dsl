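; Blueprint subclass that plays music party audio for the Ferris wheel speakers, extending its parent's start and stop events.
; On start it calls the parent, then, if the Audio component is valid, sends it post-effect to the AudioBus at full level and quickly fades it in.
; On stop it passes FadeOutDuration to the parent, but fades its own component out over a short fixed time instead of that duration.
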(event EventStartMusicParty
  (|CallParentFunction)
  (bind _audio (Variables|Default|GetAudio)
    (:then
      (Audio|Components|Audio|SetAudioBusSendPostEffect _audio (Variables|Audio|GetAudioBus) 1.0 :!class /Script/Engine.AudioComponent)
      (Audio|Components|Audio|FadeIn _audio 0.1 1.0 0.0 "Linear" :!class /Script/Engine.AudioComponent))
    (:else)))

(event EventStopMusicParty
  (params (FadeOutDuration :type Float%20%28double-precision%29))
  (|CallParentFunction FadeOutDuration)
  (bind _audio (Variables|Default|GetAudio)
    (:then
      (Audio|Components|Audio|FadeOut _audio 0.1 0.0 "Linear" :!class /Script/Engine.AudioComponent))
    (:else)))

(fn ConstructionScript)
