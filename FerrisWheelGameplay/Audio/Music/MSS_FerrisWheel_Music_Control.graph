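# Stereo MetaSound source for the ferris wheel music, feeding the AB_FerrisWheelSpeakers audio bus.
# On play, it reads the wave asset's duration and starts the stereo wave player at a random time within it.
# An ADSR envelope triggered by playback applies a short fade-in to both channels, masking the mid-track start.
# Base submix and submix sends are disabled, so output goes through the bus send.
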MetaSoundSource /FerrisWheelGameplay/Audio/Music/MSS_FerrisWheel_Music_Control
inputs [UE.Source.OnPlay {type=Trigger}]
interfaces [UE.OutputFormat.Stereo:1.0, UE.Source:1.1]
outputs [
 "UE.OutputFormat.Stereo.Audio:0" {type=Audio}
 "UE.OutputFormat.Stereo.Audio:1" {type=Audio}
]
pages [Default {
  input_defaults {UE.Source.OnPlay=null}
  nodes [
   n1 = VariableAccessor.WaveAsset(Variable=InitVariable.WaveAsset().Variable)
   GetWaveDuration_0 = "UE.Get Wave Duration: "(Wave=@n1.Value)
   n2 = UE.RandomTime(Next=^UE.Source.OnPlay, Max=@GetWaveDuration_0.Duration)
   WavePlayer_0 = UE.Wave Player:Stereo(Play=@n2.On Next, Start Time=@n2.Value, Wave Asset=@n1.Value) {also {defaults {Loop="<MetaSoundBuilderResult.SUCCEEDED: 0>"}, version [1,1]}}
   ADSREnvelopeADSREnvelope_0 = ADSR Envelope.ADSR Envelope:Float(Trigger Attack=@WavePlayer_0.On Play) {also {defaults {Attack Time=0.125, Sustain Level=1}}}
   Multiply_0 = UE.Multiply:Audio by Float(AdditionalOperands=@ADSREnvelopeADSREnvelope_0.Out Envelope, PrimaryOperand=@WavePlayer_0.Out Left)
   Multiply_1 = UE.Multiply:Audio by Float(AdditionalOperands=@ADSREnvelopeADSREnvelope_0.Out Envelope, PrimaryOperand=@WavePlayer_0.Out Right)
  ]
  links [{dst="@UE.OutputFormat.Stereo.Audio:0", src=Multiply_0.Out}, {dst="@UE.OutputFormat.Stereo.Audio:1", src=Multiply_1.Out}]
 }]
properties {
 assetImportData=&self:AssetImportData
 bEnableBaseSubmix=false
 bEnableSubmixSends=false
 bLooping=true
 duration=10000
 numChannels=2
 outputFormat=Stereo
 sampleRate=48000
 soundClassObject=&Fort_Ambient_Athena
 busSends [{audioBus=&/AB_FerrisWheelSpeakers, sendLevel=1}]
}
